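# Control Rig for the male medium Bicycle body that procedurally corrects shirt sleeve bones as the arms move.
# Upper-arm twist sine drives signed sleeve offsets and scales; arm rotation and hand-to-goal distance add further offsets.
# Dot-product pose readers compare upper-arm direction with goal controls, nudging mostly right sleeve bones; the sole left reader reuses right-side references.
# Debug, arm-pose and Kalman filter nodes sit outside the executed chain.
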ControlRigBlueprint /BRCosmetics/Characters/Player/Male/Medium/Bodies/M_MED_Bicycle/Meshes/M_MED_Bicycle_CtrlRig
graphs [
 RigVMFunctionLibrary {links [], nodes []}
 RigVMModel {nodes [
   n1 = &RigUnit_BeginExecution()
   n2 = /Script/ControlRig.RigUnit_SequenceExecution(ExecuteContext=@n1.ExecutePin) {is=struct RigVMUnitNode}
   n3 = &RigUnit_SequenceExecution(ExecuteContext=@n2.A)
   RigVMVariableNode(
    Value=&RigVMFunction_MathQuaternionToEuler(
     RotationOrder=ZYX
     Value=&RigUnit_GetRelativeTransformForItem(Child=(Type=Bone,Name="upperarm_twist_01_r"), Parent=(Type=Bone,Name="upperarm_r")).RelativeTransform.Rotation
    ).Result.X
    ExecuteContext=@n3.A
   ) {variable=shldr_twist_deg_r}
   n4 = &RigUnit_SequenceExecution(ExecuteContext=@n3.D)
   RigVMVariableNode(
    Value=&RigVMFunction_MathQuaternionToEuler(
     RotationOrder=ZYX
     Value=&RigUnit_GetRelativeTransformForItem(Child=(Type=Bone,Name="upperarm_twist_01_l"), Parent=(Type=Bone,Name="upperarm_l")).RelativeTransform.Rotation
    ).Result.X
    ExecuteContext=@n4.A
   ) {variable=shldr_twist_deg_l}
   n5 = /Script/RigVM.RigVMFunction_MathFloatSin(Value=/Script/RigVM.RigVMFunction_MathFloatRad(Value=RigVMVariableNode() {variable=shldr_twist_deg_r}.Value) {is=struct RigVMUnitNode, template="Radians::Execute(in Value,out Result)"}.Result) {is=struct RigVMUnitNode, template="Sin::Execute(in Value,out Result)"}
   &RigUnit_OffsetTransformForItem(
    Item=(Type=Bone,Name="sleeve_r")
    OffsetTransform=(Rotation=(X=0,Y=0,Z=0,W=1),Translation=(X=0,Y=2,Z=-2),Scale3D=(X=1,Y=0.9,Z=1.35))
    bPropagateToChildren=True
    Weight=/Script/RigVM.RigVMFunction_MathFloatClamp(Maximum=1, Value=@n5.Result) {is=struct RigVMUnitNode, template="Clamp::Execute(in Value,in Minimum,in Maximum,out Result)"}.Result
    ExecutePin=@n2.B
   )
   > &RigUnit_OffsetTransformForItem(
    Item=(Type=Bone,Name="sleeve_r")
    OffsetTransform=(Rotation=(X=0,Y=0,Z=0,W=1),Translation=(X=0,Y=-2,Z=-2),Scale3D=(X=1,Y=1,Z=1.2))
    bPropagateToChildren=True
    Weight=&RigVMFunction_MathFloatClamp(Maximum=1, Value=&RigVMFunction_MathFloatMul(B=-1, A=@n5.Result) {template="Multiply::Execute(in A,in B,out Result)"}.Result) {template="Clamp::Execute(in Value,in Minimum,in Maximum,out Result)"}.Result
   )
   n6 = &RigVMFunction_MathFloatSin(Value=&RigVMFunction_MathFloatRad(Value=RigVMVariableNode() {variable=shldr_twist_deg_l}.Value) {template="Radians::Execute(in Value,out Result)"}.Result) {template="Sin::Execute(in Value,out Result)"}
   > &RigUnit_OffsetTransformForItem(
    Item=(Type=Bone,Name="sleeve_l")
    OffsetTransform=(Rotation=(X=0,Y=0,Z=0,W=1),Translation=(X=0,Y=2,Z=2),Scale3D=(X=1,Y=0.9,Z=1.44))
    bPropagateToChildren=True
    Weight=&RigVMFunction_MathFloatClamp(Maximum=1, Value=@n6.Result) {template="Clamp::Execute(in Value,in Minimum,in Maximum,out Result)"}.Result
   )
   > &RigUnit_OffsetTransformForItem(
    Item=(Type=Bone,Name="sleeve_l")
    OffsetTransform=(Rotation=(X=0,Y=0,Z=0,W=1),Translation=(X=0,Y=-2,Z=1.4),Scale3D=(X=1,Y=1,Z=1.2))
    bPropagateToChildren=True
    Weight=&RigVMFunction_MathFloatClamp(Maximum=1, Value=&RigVMFunction_MathFloatMul(B=-1, A=@n6.Result) {template="Multiply::Execute(in A,in B,out Result)"}.Result) {template="Clamp::Execute(in Value,in Minimum,in Maximum,out Result)"}.Result
   )
   n7 = /Script/RigVM.RigVMFunction_ForLoopCount(Count=4, ExecutePin=@n2.D) {is=struct RigVMUnitNode}
   n8 = &RigUnit_GetRelativeTransformForItem(Child=(Type=Bone,Name="upperarm_r"), Parent=(Type=Bone,Name="spine_05"))
   n9 = &RigUnit_GetRelativeTransformForItem(Child=(Type=Bone,Name="lowerarm_r"), Parent=(Type=Bone,Name="spine_05"))
   n10 = &RigUnit_GetRelativeTransformForItem(Child=(Type=Bone,Name="upperarm_r"), Parent=(Type=Bone,Name="spine_05"), bChildInitial=true, bParentInitial=true)
   RigUnit_CollectionItems = /Script/ControlRig.RigUnit_CollectionItems(Items=((Type=Control,Name="R_elbow_celebrate_GOAL"),(Type=Control,Name="R_elbow_fwd_GOAL"),(Type=Control,Name="R_elbow_side_GOAL"),(Type=Control,Name="R_elbow_side_GOAL_2"))) {is=struct RigVMUnitNode}
   n11 = &RigUnit_GetRelativeTransformForItem(
    Parent=(Type=Bone,Name="spine_05")
    bChildInitial=true
    bParentInitial=true
    Child=/Script/ControlRig.RigUnit_CollectionItemAtIndex(Index=@n7.Index, Collection=@RigUnit_CollectionItems.Collection) {is=struct RigVMUnitNode}.Item
   )
   n12 = /Script/RigVM.RigVMFunction_MathVectorDot(
    A=/Script/RigVM.RigVMFunction_MathVectorUnit(Value=/Script/RigVM.RigVMFunction_MathVectorSub(B=@n8.RelativeTransform.Translation, A=@n9.RelativeTransform.Translation) {is=struct RigVMUnitNode, template="Subtract::Execute(in A,in B,out Result)"}.Result) {is=struct RigVMUnitNode, template="Unit::Execute(in Value,out Result)"}.Result
    B=&RigVMFunction_MathVectorUnit(Value=&RigVMFunction_MathVectorSub(B=@n10.RelativeTransform.Translation, A=@n11.RelativeTransform.Translation) {template="Subtract::Execute(in A,in B,out Result)"}.Result) {template="Unit::Execute(in Value,out Result)"}.Result
   ) {is=struct RigVMUnitNode, template="Dot,|::Execute(in A,in B,out Result)"}
   RigUnit_AlphaInterp_1 = &RigVMFunction_AlphaInterp(
    InRange=(Min=0.6,Max=0.9)
    InterpSpeedDecreasing=10
    InterpSpeedIncreasing=10
    Scale=1
    bClampResult=true
    bMapRange=true
    Bias=~
    ClampMin=~
    bInterpResult=~
    Value=@n12.Result
   )
   > &RigUnit_ModifyTransforms(
    ItemToModify=((Item=(Type=Bone,Name="sleeve_r"),Transform=(Rotation=(X=0,Y=0,Z=0,W=1),Translation=(X=0,Y=0,Z=1),Scale3D=(X=1,Y=1,Z=1))))
    ItemToModify.0.Transform=/Script/RigVM.RigVMFunction_MathTransformFromSRT(
     Rotation=(X=0,Y=0,Z=0)
     Scale=(X=1,Y=1,Z=1)
     Location=RigVMDispatchNode(Values=((X=0,Y=0,Z=1),(X=0,Y=0,Z=1),(X=0,Y=0,Z=1),(X=0,Y=0,Z=0.2)), Index=@n7.Index) {template="DISPATCH_RigVMDispatch_SelectInt32(in Index,in Values,out Result)"}.Result
    ) {is=struct RigVMUnitNode}.Transform
    Weight=@RigUnit_AlphaInterp_1.Result
   )
   n13 = &RigUnit_SequenceExecution(ExecuteContext=@n7.Completed)
   n14 = &RigVMFunction_ForLoopCount(Count=2, ExecutePin=@n13.A)
   n15 = &RigUnit_GetRelativeTransformForItem(Child=(Type=Bone,Name="upperarm_r"), Parent=(Type=Bone,Name="spine_05"), bChildInitial=true, bParentInitial=true)
   n16 = &RigUnit_GetRelativeTransformForItem(
    Parent=(Type=Bone,Name="spine_05")
    bChildInitial=true
    bParentInitial=true
    Child=&RigUnit_CollectionItemAtIndex(
     Index=@n14.Index
     Collection=&RigUnit_CollectionItems(Items=((Type=Control,Name="R_elbow_downFwd_GOAL"),(Type=Control,Name="R_elbow_side_GOAL_2"))).Collection
    ).Item
   )
   n17 = &RigUnit_GetRelativeTransformForItem(Child=(Type=Bone,Name="upperarm_r"), Parent=(Type=Bone,Name="spine_05"))
   n18 = &RigUnit_GetRelativeTransformForItem(Child=(Type=Bone,Name="lowerarm_r"), Parent=(Type=Bone,Name="spine_05"))
   n19 = &RigVMFunction_MathVectorDot(
    B=&RigVMFunction_MathVectorUnit(Value=&RigVMFunction_MathVectorSub(B=@n15.RelativeTransform.Translation, A=@n16.RelativeTransform.Translation) {template="Subtract::Execute(in A,in B,out Result)"}.Result) {template="Unit::Execute(in Value,out Result)"}.Result
    A=&RigVMFunction_MathVectorUnit(Value=&RigVMFunction_MathVectorSub(B=@n17.RelativeTransform.Translation, A=@n18.RelativeTransform.Translation) {template="Subtract::Execute(in A,in B,out Result)"}.Result) {template="Unit::Execute(in Value,out Result)"}.Result
   ) {template="Dot,|::Execute(in A,in B,out Result)"}
   RigUnit_AlphaInterp_1_1_2 = &RigVMFunction_AlphaInterp(
    InRange=(Min=0.6,Max=0.9)
    InterpSpeedDecreasing=10
    InterpSpeedIncreasing=10
    Scale=1
    bClampResult=true
    bMapRange=true
    Bias=~
    ClampMin=~
    bInterpResult=~
    Value=@n19.Result
   )
   > &RigUnit_ModifyTransforms(
    ItemToModify=((Item=(Type=Bone,Name="sleeve_in_r"),Transform=(Rotation=(X=0,Y=0,Z=0,W=1),Translation=(X=0,Y=0,Z=1),Scale3D=(X=1,Y=1,Z=1))))
    ItemToModify.0.Transform=&RigVMFunction_MathTransformFromSRT(
     Rotation=(X=0,Y=0,Z=0)
     Scale=(X=1,Y=1,Z=1)
     Location=RigVMDispatchNode(Values=((X=0,Y=0,Z=0),(X=1,Y=5,Z=-5)), Index=@n14.Index) {template="DISPATCH_RigVMDispatch_SelectInt32(in Index,in Values,out Result)"}.Result
    ).Transform
    Weight=@RigUnit_AlphaInterp_1_1_2.Result
   )
   n20 = &RigVMFunction_ForLoopCount(Count=2, ExecutePin=@n13.B)
   n21 = &RigUnit_GetRelativeTransformForItem(Child=(Type=Bone,Name="upperarm_r"), Parent=(Type=Bone,Name="spine_05"), bChildInitial=true, bParentInitial=true)
   n22 = &RigUnit_GetRelativeTransformForItem(
    Parent=(Type=Bone,Name="spine_05")
    bChildInitial=true
    bParentInitial=true
    Child=&RigUnit_CollectionItemAtIndex(
     Index=@n20.Index
     Collection=&RigUnit_CollectionItems(Items=((Type=Control,Name="R_elbow_downFwd_GOAL"),(Type=Control,Name="R_elbow_side_GOAL_2"))).Collection
    ).Item
   )
   n23 = &RigUnit_GetRelativeTransformForItem(Child=(Type=Bone,Name="upperarm_r"), Parent=(Type=Bone,Name="spine_05"))
   n24 = &RigUnit_GetRelativeTransformForItem(Child=(Type=Bone,Name="lowerarm_r"), Parent=(Type=Bone,Name="spine_05"))
   n25 = &RigVMFunction_MathVectorDot(
    B=&RigVMFunction_MathVectorUnit(Value=&RigVMFunction_MathVectorSub(B=@n21.RelativeTransform.Translation, A=@n22.RelativeTransform.Translation) {template="Subtract::Execute(in A,in B,out Result)"}.Result) {template="Unit::Execute(in Value,out Result)"}.Result
    A=&RigVMFunction_MathVectorUnit(Value=&RigVMFunction_MathVectorSub(B=@n23.RelativeTransform.Translation, A=@n24.RelativeTransform.Translation) {template="Subtract::Execute(in A,in B,out Result)"}.Result) {template="Unit::Execute(in Value,out Result)"}.Result
   ) {template="Dot,|::Execute(in A,in B,out Result)"}
   RigUnit_AlphaInterp_1_1 = &RigVMFunction_AlphaInterp(
    InRange=(Min=0.6,Max=0.9)
    InterpSpeedDecreasing=10
    InterpSpeedIncreasing=10
    Scale=1
    bClampResult=true
    bMapRange=true
    Bias=~
    ClampMin=~
    bInterpResult=~
    Value=@n25.Result
   )
   > &RigUnit_ModifyTransforms(
    ItemToModify=((Item=(Type=Bone,Name="sleeve_offset_r"),Transform=(Rotation=(X=0,Y=0,Z=0,W=1),Translation=(X=0,Y=0,Z=1),Scale3D=(X=1,Y=1,Z=1))))
    ItemToModify.0.Transform=&RigVMFunction_MathTransformFromSRT(
     Rotation=(X=0,Y=0,Z=0)
     Scale=(X=1,Y=1,Z=1)
     Location=RigVMDispatchNode(Values=((X=0,Y=10,Z=0),(X=0,Y=-5,Z=5)), Index=@n20.Index) {template="DISPATCH_RigVMDispatch_SelectInt32(in Index,in Values,out Result)"}.Result
    ).Transform
    Weight=@RigUnit_AlphaInterp_1_1.Result
   )
   n26 = &RigVMFunction_ForLoopCount(Count=1, ExecutePin=@n20.Completed)
   n27 = &RigUnit_GetRelativeTransformForItem(Child=(Type=Bone,Name="upperarm_r"), Parent=(Type=Bone,Name="spine_05"), bChildInitial=true, bParentInitial=true)
   n28 = &RigUnit_GetRelativeTransformForItem(
    Parent=(Type=Bone,Name="spine_05")
    bChildInitial=true
    bParentInitial=true
    Child=&RigUnit_CollectionItemAtIndex(Index=@n26.Index, Collection=&RigUnit_CollectionItems(Items=((Type=Control,Name="R_elbow_downFwd_GOAL"))).Collection).Item
   )
   n29 = &RigUnit_GetRelativeTransformForItem(Child=(Type=Bone,Name="upperarm_l"), Parent=(Type=Bone,Name="spine_05"))
   n30 = &RigUnit_GetRelativeTransformForItem(Child=(Type=Bone,Name="lowerarm_l"), Parent=(Type=Bone,Name="spine_05"))
   n31 = &RigVMFunction_MathVectorDot(
    B=&RigVMFunction_MathVectorUnit(Value=&RigVMFunction_MathVectorSub(B=@n27.RelativeTransform.Translation, A=@n28.RelativeTransform.Translation) {template="Subtract::Execute(in A,in B,out Result)"}.Result) {template="Unit::Execute(in Value,out Result)"}.Result
    A=&RigVMFunction_MathVectorUnit(Value=&RigVMFunction_MathVectorSub(B=@n29.RelativeTransform.Translation, A=@n30.RelativeTransform.Translation) {template="Subtract::Execute(in A,in B,out Result)"}.Result) {template="Unit::Execute(in Value,out Result)"}.Result
   ) {template="Dot,|::Execute(in A,in B,out Result)"}
   RigUnit_AlphaInterp_1_1_1 = &RigVMFunction_AlphaInterp(
    InRange=(Min=0.6,Max=0.9)
    InterpSpeedDecreasing=10
    InterpSpeedIncreasing=10
    Scale=1
    bClampResult=true
    bMapRange=true
    Bias=~
    ClampMin=~
    bInterpResult=~
    Value=@n31.Result
   )
   > &RigUnit_ModifyTransforms(
    ItemToModify=((Item=(Type=Bone,Name="sleeve_offset_l"),Transform=(Rotation=(X=0,Y=0,Z=0,W=1),Translation=(X=0,Y=0,Z=1),Scale3D=(X=1,Y=1,Z=1))))
    ItemToModify.0.Transform=&RigVMFunction_MathTransformFromSRT(
     Rotation=(X=0,Y=0,Z=0)
     Scale=(X=1,Y=1,Z=1)
     Location=RigVMDispatchNode(Values=((X=0,Y=5,Z=-5)), Index=@n26.Index) {template="DISPATCH_RigVMDispatch_SelectInt32(in Index,in Values,out Result)"}.Result
    ).Transform
    Weight=@RigUnit_AlphaInterp_1_1_1.Result
   )
   n32 = &RigVMFunction_MathQuaternionToEuler(
    RotationOrder=ZYX
    Value=&RigUnit_GetRelativeTransformForItem(Child=(Type=Bone,Name="upperarm_r"), Parent=(Type=Bone,Name="spine_05")).RelativeTransform.Rotation
   )
   &RigUnit_OffsetTransformForItem(
    Item=(Type=Bone,Name="sleeve_offset_r")
    OffsetTransform=(Rotation=(X=0,Y=0,Z=0,W=1),Translation=(X=0,Y=15,Z=5),Scale3D=(X=1,Y=1,Z=1))
    bPropagateToChildren=True
    Weight=&RigVMFunction_MathFloatRemap(
     SourceMaximum=123
     SourceMinimum=30
     TargetMinimum=0
     bClamp=true
     Value=/Script/RigVM.RigVMFunction_MathFloatAbs(Value=@n32.Result.Z) {is=struct RigVMUnitNode, template="Absolute::Execute(in Value,out Result)"}.Result
    ).Result
    ExecutePin=@n2.C
   )
   n33 = &RigVMFunction_MathQuaternionToEuler(
    RotationOrder=ZYX
    Value=&RigUnit_GetRelativeTransformForItem(Child=(Type=Bone,Name="upperarm_l"), Parent=(Type=Bone,Name="spine_05")).RelativeTransform.Rotation
   )
   > &RigUnit_OffsetTransformForItem(
    Item=(Type=Bone,Name="sleeve_offset_l")
    OffsetTransform=(Rotation=(X=0,Y=0,Z=0,W=1),Translation=(X=0,Y=10,Z=0),Scale3D=(X=1,Y=1,Z=1))
    bPropagateToChildren=True
    Weight=&RigVMFunction_MathFloatRemap(
     SourceMaximum=50
     SourceMinimum=170
     TargetMinimum=0
     bClamp=true
     Value=&RigVMFunction_MathFloatAbs(Value=@n33.Result.Z) {template="Absolute::Execute(in Value,out Result)"}.Result
    ).Result
   )
   n34 = &RigVMFunction_AlphaInterp(
    InRange=(Min=10,Max=2)
    InterpSpeedDecreasing=10
    InterpSpeedIncreasing=10
    Scale=1
    bClampResult=true
    bMapRange=true
    Bias=~
    ClampMin=~
    bInterpResult=~
    Value=/Script/RigVM.RigVMFunction_MathVectorLength(Value=&RigUnit_GetRelativeTransformForItem(Child=(Type=Control,Name="L_hand_FE_GOAL"), Parent=(Type=Bone,Name="hand_l")).RelativeTransform.Translation) {is=struct RigVMUnitNode, template="Length::Execute(in Value,out Result)"}.Result
   )
   > &RigUnit_OffsetTransformForItem(
    Item=(Type=Bone,Name="sleeve_offset_l")
    OffsetTransform=(Rotation=(X=0,Y=0,Z=0,W=1),Translation=(X=0,Y=3,Z=0),Scale3D=(X=1,Y=1,Z=1))
    bPropagateToChildren=True
    Weight=@n34.Result
   )
   > &RigUnit_OffsetTransformForItem(
    Item=(Type=Bone,Name="sleeve_l")
    OffsetTransform=(Rotation=(X=0,Y=0,Z=0,W=1),Translation=(X=0,Y=-1,Z=0),Scale3D=(X=1,Y=1.1,Z=1))
    bPropagateToChildren=True
    Weight=@n34.Result
   )
   n35 = &RigVMFunction_AlphaInterp(
    InRange=(Min=10,Max=2)
    InterpSpeedDecreasing=10
    InterpSpeedIncreasing=10
    Scale=1
    bClampResult=true
    bMapRange=true
    Bias=~
    ClampMin=~
    bInterpResult=~
    Value=&RigVMFunction_MathVectorLength(Value=&RigUnit_GetRelativeTransformForItem(Child=(Type=Control,Name="R_hand_FE_GOAL"), Parent=(Type=Bone,Name="hand_r")).RelativeTransform.Translation) {template="Length::Execute(in Value,out Result)"}.Result
   )
   > &RigUnit_OffsetTransformForItem(
    Item=(Type=Bone,Name="sleeve_offset_r")
    OffsetTransform=(Rotation=(X=0,Y=0,Z=0,W=1),Translation=(X=0,Y=10,Z=10),Scale3D=(X=1,Y=1,Z=1))
    bPropagateToChildren=True
    Weight=@n35.Result
   )
   n36 = &RigUnit_SequenceExecution(ExecuteContext=())
   &RigUnit_SetTransform(
    Item=(Type=Bone,Name="upperarm_l")
    Value=(Rotation=(X=-0.0134,Y=0.3555,Z=-0.014,W=0.9345),Translation=(X=17.7788,Y=-4.8444,Z=134.4501),Scale3D=(X=1,Y=1,Z=1))
    Weight=1
    ExecutePin=@n36.A
   )
   &RigUnit_SetTransform(
    Item=(Type=Bone,Name="upperarm_r")
    Value=(Rotation=(X=-0.9345,Y=-0.014,Z=-0.3555,W=-0.0134),Translation=(X=-17.7788,Y=-4.8444,Z=134.4503),Scale3D=(X=1,Y=1,Z=1))
    Weight=1
    ExecutePin=@n36.C
   )
   > /Script/ControlRig.RigUnit_DebugTransformMutableItemSpace(
    Color=(R=1,G=0,B=0,A=1)
    DebugDrawSettings=(DepthPriority=SDPG_Foreground,Lifetime=-1)
    Mode=Axes
    Scale=10
    Space=(Type=Bone)
    Transform=(Rotation=(X=0,Y=0,Z=0,W=1),Translation=(X=0,Y=0,Z=0),Scale3D=(X=1,Y=1,Z=1))
    WorldOffset=(Rotation=(X=0,Y=0,Z=0,W=1),Translation=(X=0,Y=0,Z=0),Scale3D=(X=1,Y=1,Z=1))
    bEnabled=True
    Transform.Translation=&RigVMFunction_MathQuaternionToEuler(RotationOrder=ZYX, Value=&RigUnit_GetTransform(Item=(Type=Bone,Name="upperarm_r")).Transform.Rotation).Result
   ) {is=struct RigVMUnitNode}
   n37 = /Script/ControlRig.RigUnit_ProjectTransformToNewParent(
    Child=(Type=Bone,Name="upperarm_twist_01_l")
    NewParent=(Type=Bone,Name="clavicle_l")
    OldParent=(Type=Bone,Name="upperarm_l")
    bChildInitial=True
    bNewParentInitial=false
    bOldParentInitial=True
   ) {is=struct RigVMUnitNode}
   &RigUnit_SetRotation(ExecutePin=(), Item=(Type=Bone,Name="upperarm_twist_01_l"), Weight=1, Value=@n37.Transform.Rotation)
   > /Script/ControlRig.RigUnit_AimItem(
    DebugSettings=(Scale=10,WorldOffset=(Rotation=(X=0,Y=0,Z=0,W=1),Translation=(X=0,Y=0,Z=0),Scale3D=(X=1,Y=1,Z=1)))
    Item=(Type=Bone,Name="upperarm_twist_01_l")
    Primary=(Weight=1,Axis=(X=1,Y=0,Z=0),Target=(X=0,Y=0,Z=0),Kind=Location,Space=(Type=Bone,Name="None"))
    Secondary=(Weight=0,Axis=(X=0,Y=0,Z=1),Target=(X=1,Y=0,Z=0),Kind=Location,Space=(Type=Bone))
    Weight=0
    Primary.Axis=&RigUnit_GetRelativeTransformForItem(Child=(Type=Bone,Name="lowerarm_l"), Parent=(Type=Bone,Name="upperarm_twist_01_l"), bChildInitial=true, bParentInitial=true).RelativeTransform.Translation
    Primary.Target=&RigUnit_GetTransform(Item=(Type=Bone,Name="lowerarm_l")).Transform.Translation
   ) {is=struct RigVMUnitNode}
   > &RigUnit_DebugTransformMutableItemSpace(
    Color=(R=1,G=0,B=0,A=1)
    DebugDrawSettings=(DepthPriority=SDPG_Foreground,Lifetime=-1)
    Mode=Axes
    Scale=10
    Space=(Type=Bone)
    WorldOffset=(Rotation=(X=0,Y=0,Z=0,W=1),Translation=(X=0,Y=0,Z=0),Scale3D=(X=1,Y=1,Z=1))
    bEnabled=True
    Transform=&RigUnit_GetTransform(Item=(Type=Bone,Name="upperarm_twist_01_l")).Transform
   )
   &RigVMFunction_KalmanFloat(BufferSize=16, Value=@n12.Result) {template="KalmanFilter::Execute(in Value,in BufferSize,out Result)"}
   &RigVMFunction_KalmanFloat(BufferSize=16, Value=@n25.Result) {template="KalmanFilter::Execute(in Value,in BufferSize,out Result)"}
   &RigVMFunction_KalmanFloat(BufferSize=16, Value=@n31.Result) {template="KalmanFilter::Execute(in Value,in BufferSize,out Result)"}
   &RigVMFunction_KalmanFloat(BufferSize=16, Value=@n19.Result) {template="KalmanFilter::Execute(in Value,in BufferSize,out Result)"}
   &RigVMFunction_MathFloatAdd(A=5) {template="Add::Execute(in A,in B,out Result)"}
   &RigUnit_ProjectTransformToNewParent(
    Child=(Type=Bone,Name="upperarm_twist_01_l")
    NewParent=(Type=Bone,Name="clavicle_l")
    OldParent=(Type=Bone,Name="upperarm_l")
    bChildInitial=True
    bNewParentInitial=false
    bOldParentInitial=True
   )
   &RigUnit_GetTransform(Item=(Type=Bone,Name="upperarm_twist_01_l"))
   RigUnit_SetTransform = &RigUnit_SetTransform(
    ExecutePin=()
    Item=(Type=Bone,Name="upperarm_twist_01_l")
    Value=(Rotation=(X=0.0134,Y=-0.3555,Z=0.014,W=-0.9345),Translation=(X=17.7819,Y=-4.8445,Z=134.4474),Scale3D=(X=1,Y=1,Z=1))
    Weight=1
    bInitial=false
    Value.Rotation=&RigVMFunction_MathQuaternionFromEuler(Euler=(X=0,Y=0,Z=0)).Result
   )
  ]}
]
hierarchy [
 root {loc=<0 0 0>, rot=<0 0 0 1>}
 pelvis {parent=root, loc=<0 -0.7071 93.1191>, rot=<0.0127 0.707 -0.0127 -0.707>}
 spine_01 {parent=pelvis, loc=<6.7494 -0.6388 0>, rot=<0 0 0.034 -0.9994>}
 spine_02 {parent=spine_01, loc=<7.2518 -0.035 0>, rot=<0 0 0.0238 -0.9997>}
 spine_03 {parent=spine_02, loc=<6.0104 0 0>, rot=<0 0 0.0149 -0.9999>}
 spine_04 {parent=spine_03, loc=<8.1024 0 0>, rot=<0 0 -0.0132 -0.9999>}
 spine_05 {parent=spine_04, loc=<10.0146 0 0>, rot=<0 0 -0.0015 -1>}
 clavicle_l {parent=spine_05, loc=<4.3575 -0.1466 -2.3223>, rot=<0.0465 -0.7289 -0.009 -0.6829>}
 upperarm_l {parent=clavicle_l, loc=<15.5111 0 0>, rot=<0.0226 -0.3234 -0.0088 -0.9459>}
 lowerarm_l {parent=upperarm_l, loc=<26.6514 0 0>, rot=<0.0391 -0.0099 -0.2453 -0.9686>}
 hand_l {parent=lowerarm_l, loc=<22.3491 0 0>, rot=<-0.7081 -0.0507 0.0975 -0.6975>}
 index_metacarpal_l {parent=hand_l, loc=<1.6535 0.0183 -2.96>, rot=<0.0158 -0.0124 -0.0173 -0.9996>}
 index_01_l {parent=index_metacarpal_l, loc=<8.3008 0.0357 -0.0524>, rot=<-0.0044 0.0485 0.0924 -0.9945>}
 index_02_l {parent=index_01_l, loc=<3.3587 -0.3004 0.0452>, rot=<-0.002 0.0027 0.3225 -0.9466>}
 index_03_l {parent=index_02_l, loc=<2.5578 -0.123 -0.0252>, rot=<-0.002 0.0067 0.2558 -0.9667>}
 middle_metacarpal_l {parent=hand_l, loc=<2.1856 -0.1725 -0.5984>, rot=<0.006 0.0168 -0.0329 -0.9993>}
 middle_01_l {parent=middle_metacarpal_l, loc=<7.3275 -0.2447 0.0129>, rot=<-0.0208 0.0042 0.0858 -0.9961>}
 middle_02_l {parent=middle_01_l, loc=<4.0123 0.1401 0.1175>, rot=<-0.0218 0.0258 0.3512 -0.9357>}
 middle_03_l {parent=middle_02_l, loc=<2.6077 0.0534 0.1373>, rot=<0.0082 0.0188 0.1708 -0.9851>}
 pinky_metacarpal_l {parent=hand_l, loc=<1.7616 -0.9753 3.3814>, rot=<-0.0397 0.0814 -0.069 -0.9935>}
 pinky_01_l {parent=pinky_metacarpal_l, loc=<7.1709 -0.4831 -0.1094>, rot=<-0.0654 -0.0035 0.2242 -0.9723>}
 pinky_02_l {parent=pinky_01_l, loc=<2.8274 0.0893 -0.1043>, rot=<0.0037 0.0011 0.2854 -0.9584>}
 pinky_03_l {parent=pinky_02_l, loc=<2.2046 0.0308 -0.0078>, rot=<-0.0002 0.0098 0.1445 -0.9895>}
 ring_metacarpal_l {parent=hand_l, loc=<2.2074 -0.3625 1.3638>, rot=<-0.0284 0.0722 -0.0458 -0.9959>}
 ring_01_l {parent=ring_metacarpal_l, loc=<7.0295 -0.6197 -0.1446>, rot=<-0.0489 -0.0252 0.1892 -0.9804>}
 ring_02_l {parent=ring_01_l, loc=<3.6174 0.4647 -0.1027>, rot=<0 0 0.2806 -0.9598>}
 ring_03_l {parent=ring_02_l, loc=<2.4152 0.114 0.0199>, rot=<0 0 0.1189 -0.9929>}
 thumb_01_l {parent=hand_l, loc=<1.3118 -0.3191 -2.459>, rot=<0.6546 -0.4042 0.0664 -0.6354>}
 thumb_02_l {parent=thumb_01_l, loc=<5.9336 -0.0448 0.0401>, rot=<-0.0037 0.0213 0.298 -0.9543>}
 thumb_03_l {parent=thumb_02_l, loc=<2.2223 0.3682 -0.0907>, rot=<0.0118 -0.013 0.0066 -0.9998>}
 lowerarm_twist_01_l {parent=lowerarm_l, loc=<17.8549 0 0>, rot=<0 0 0 -1>}
 lowerarm_twist_02_l {parent=lowerarm_l, loc=<9.6199 0 0>, rot=<0 0 0 -1>}
 deform_elbow_bk_l {parent=lowerarm_l, loc=<-1.2853 -2.9274 -0.5607>, rot=<-0.0285 0.0335 -0.0024 -0.999>}
 upperarm_twist_01_l {parent=upperarm_l, loc=<0.0041 0 0>, rot=<0 0 0 -1>}
 sleeve_l {parent=upperarm_twist_01_l, loc=<-0.0001 0 0.0001>, rot=<0 0 1 0>}
 sleeve_offset_l {parent=sleeve_l, loc=<-3.4359 -4.8798 -0.3802>, rot=<0.1933 0.1022 0.9678 0.1244>}
 sleeve_in_l {parent=sleeve_l, loc=<-3.8376 3.8583 -6.4048>, rot=<0.0243 -0.1622 0.9711 0.1735>}
 deform_deltoid_fr_l {parent=upperarm_twist_01_l, loc=<2.4511 0 0>, rot=<0 0 0 -1>}
 upperarm_twist_02_l {parent=upperarm_l, loc=<13.7941 0 0>, rot=<0 0 0 -1>}
 deform_elbow_fr_l {parent=upperarm_l, loc=<26.5048 0.1875 -0.5816>, rot=<0.0021 0.0156 -0.185 -0.9826>}
 forearm_spike_twist_l {parent=upperarm_l, loc=<26.6514 0 0.0002>, rot=<0.0099 0.0391 0.9686 -0.2453>}
 forearm_spike_in_l {parent=upperarm_l, loc=<26.6514 0 0.0001>, rot=<0.0099 0.0391 0.9686 -0.2453>}
 forearm_spike_tip_l {parent=upperarm_l, loc=<26.6514 0 0.0002>, rot=<0 0 1 0>}
 forearm_spike_l {parent=upperarm_l, loc=<26.6514 0 0.0002>, rot=<0 0 1 0>}
 deform_pectoral_fr_l {parent=clavicle_l, loc=<9.3905 0 0>, rot=<0 0 0 -1>}
 clavicle_r {parent=spine_05, loc=<4.3576 -0.1466 2.3223>, rot=<0.7289 0.0465 -0.6829 0.009>}
 upperarm_r {parent=clavicle_r, loc=<-15.5111 0 0>, rot=<0.0226 -0.3234 -0.0088 -0.9459>}
 lowerarm_r {parent=upperarm_r, loc=<-26.6515 0 0.0001>, rot=<0.0391 -0.0099 -0.2453 -0.9686>}
 hand_r {parent=lowerarm_r, loc=<-22.3492 0 0>, rot=<-0.7081 -0.0507 0.0975 -0.6975>}
 index_metacarpal_r {parent=hand_r, loc=<-1.6534 -0.0181 2.9601>, rot=<0.0158 -0.0124 -0.0173 -0.9996>}
 index_01_r {parent=index_metacarpal_r, loc=<-8.3008 -0.0357 0.0524>, rot=<-0.0044 0.0485 0.0924 -0.9945>}
 index_02_r {parent=index_01_r, loc=<-3.3589 0.3003 -0.0453>, rot=<-0.002 0.0027 0.3225 -0.9466>}
 index_03_r {parent=index_02_r, loc=<-2.5578 0.1231 0.0251>, rot=<-0.002 0.0067 0.2558 -0.9667>}
 middle_metacarpal_r {parent=hand_r, loc=<-2.1856 0.1726 0.5984>, rot=<0.006 0.0168 -0.0329 -0.9993>}
 middle_01_r {parent=middle_metacarpal_r, loc=<-7.3273 0.2447 -0.0129>, rot=<-0.0208 0.0042 0.0858 -0.9961>}
 middle_02_r {parent=middle_01_r, loc=<-4.0142 -0.1411 -0.07>, rot=<-0.0217 0.0259 0.3512 -0.9357>}
 middle_03_r {parent=middle_02_r, loc=<-2.6074 -0.0534 -0.1312>, rot=<0.0083 0.0188 0.1708 -0.9851>}
 pinky_metacarpal_r {parent=hand_r, loc=<-1.7617 0.9753 -3.3814>, rot=<-0.0397 0.0814 -0.069 -0.9935>}
 pinky_01_r {parent=pinky_metacarpal_r, loc=<-7.1707 0.4832 0.1094>, rot=<-0.0654 -0.0035 0.2242 -0.9723>}
 pinky_02_r {parent=pinky_01_r, loc=<-2.8274 -0.0892 0.1043>, rot=<0.0037 0.0011 0.2854 -0.9584>}
 pinky_03_r {parent=pinky_02_r, loc=<-2.2045 -0.031 0.0078>, rot=<-0.0002 0.0098 0.1445 -0.9895>}
 ring_metacarpal_r {parent=hand_r, loc=<-2.2073 0.3627 -1.3638>, rot=<-0.0284 0.0722 -0.0458 -0.9959>}
 ring_01_r {parent=ring_metacarpal_r, loc=<-7.0297 0.6197 0.1446>, rot=<-0.0489 -0.0252 0.1892 -0.9804>}
 ring_02_r {parent=ring_01_r, loc=<-3.6172 -0.4648 0.1027>, rot=<0 0 0.2806 -0.9598>}
 ring_03_r {parent=ring_02_r, loc=<-2.4153 -0.1141 -0.0199>, rot=<0 0 0.1189 -0.9929>}
 thumb_01_r {parent=hand_r, loc=<-1.3119 0.3192 2.459>, rot=<0.6546 -0.4042 0.0664 -0.6354>}
 thumb_02_r {parent=thumb_01_r, loc=<-5.9336 0.0447 -0.0401>, rot=<-0.0037 0.0213 0.298 -0.9543>}
 thumb_03_r {parent=thumb_02_r, loc=<-2.2223 -0.3682 0.0907>, rot=<0.0118 -0.013 0.0066 -0.9998>}
 lowerarm_twist_01_r {parent=lowerarm_r, loc=<-17.855 0 0>, rot=<0 0 0 -1>}
 lowerarm_twist_02_r {parent=lowerarm_r, loc=<-9.6199 0 0>, rot=<0 0 0 -1>}
 deform_elbow_bk_r {parent=lowerarm_r, loc=<1.2852 2.9274 0.561>, rot=<-0.0285 0.0335 -0.0024 -0.999>}
 upperarm_twist_01_r {parent=upperarm_r, loc=<-0.0041 0 0>, rot=<0 0 0 -1>}
 sleeve_r {parent=upperarm_twist_01_r, loc=<0 0 0>, rot=<0 0 0 -1>}
 sleeve_offset_r {parent=sleeve_r, loc=<-3.4359 -4.8798 0.3803>, rot=<-0.1933 -0.1022 0.9678 0.1244>}
 sleeve_in_r {parent=sleeve_r, loc=<-3.8377 3.8583 6.4049>, rot=<0.9711 -0.1735 0.0243 0.1622>}
 deform_deltoid_fr_r {parent=upperarm_twist_01_r, loc=<-2.4512 0 0>, rot=<0 0 0 -1>}
 upperarm_twist_02_r {parent=upperarm_r, loc=<-13.7942 0 0>, rot=<0 0 0 -1>}
 deform_elbow_fr_r {parent=upperarm_r, loc=<-26.5046 -0.1875 0.5814>, rot=<0.0021 0.0156 -0.185 -0.9826>}
 forearm_spike_twist_r {parent=upperarm_r, loc=<-26.6515 0 -0.0001>, rot=<0.0391 -0.0099 -0.2453 -0.9686>}
 forearm_spike_r {parent=upperarm_r, loc=<-26.6516 0.0001 0>, rot=<0.0391 -0.0099 -0.2453 -0.9686>}
 forearm_spike_tip_r {parent=upperarm_r, loc=<-26.6515 -0.0001 0>, rot=<0.0391 -0.0099 -0.2453 -0.9686>}
 forearm_spike_in_r {parent=upperarm_r, loc=<-26.6516 0 0>, rot=<0.0391 -0.0099 -0.2453 -0.9686>}
 deform_pectoral_fr_r {parent=clavicle_r, loc=<-9.39 0 0>, rot=<0 0 0 1>}
 neck_01 {parent=spine_05, loc=<11.3138 0.1529 0>, rot=<0 0 -0.2003 -0.9797>}
 neck_02 {parent=neck_01, loc=<4.5627 0 0>, rot=<0 0 0.0504 -0.9987>}
 head {parent=neck_02, loc=<5.3914 0 0>, rot=<0 0 0.115 -0.9934>}
 R_elbow_celebrate_GOALSpace {parent=spine_05, type=Null, loc=<0 0 0>, rot=<0 0 0 1>}
 R_elbow_celebrate_GOAL {parent=R_elbow_celebrate_GOALSpace, type=Control, loc=<0 0 0>, rot=<0 0 0 1>}
 R_elbow_fwd_GOALSpace {parent=spine_05, type=Null, loc=<0 0 0>, rot=<0 0 0 1>}
 R_elbow_fwd_GOAL {parent=R_elbow_fwd_GOALSpace, type=Control, loc=<0 0 0>, rot=<0 0 0 1>}
 L_hand_FE_GOALSpace {parent=spine_05, type=Null, loc=<0 0 0>, rot=<0 0 0 1>}
 L_hand_FE_GOAL {parent=L_hand_FE_GOALSpace, type=Control, loc=<0 0 0>, rot=<0 0 0 1>}
 R_elbow_FE_GOALSpace {parent=spine_05, type=Null, loc=<0 0 0>, rot=<0 0 0 1>}
 R_hand_FE_GOAL {parent=R_elbow_FE_GOALSpace, type=Control, loc=<0 0 0>, rot=<0 0 0 1>}
 R_elbow_downFwd_GOALSpace {parent=spine_05, type=Null, loc=<0 0 0>, rot=<0 0 0 1>}
 R_elbow_downFwd_GOAL {parent=R_elbow_downFwd_GOALSpace, type=Control, loc=<0 0 0>, rot=<0 0 0 1>}
 L_elbow_downFwd_GOALSpace {parent=spine_05, type=Null, loc=<0 0 0>, rot=<0 0 0 1>}
 L_elbow_downFwd_GOAL {parent=L_elbow_downFwd_GOALSpace, type=Control, loc=<0 0 0>, rot=<0 0 0 1>}
 R_elbow_side_GOALSpace {parent=spine_05, type=Null, loc=<0 0 0>, rot=<0 0 0 1>}
 R_elbow_side_GOAL {parent=R_elbow_side_GOALSpace, type=Control, loc=<0 0 0>, rot=<0 0 0 1>}
 R_elbow_side_GOALSpace_2 {parent=spine_05, type=Null, loc=<0 0 0>, rot=<0 0 0 1>}
 R_elbow_side_GOAL_2 {parent=R_elbow_side_GOALSpace_2, type=Control, loc=<0 0 0>, rot=<0 0 0 1>}
 thigh_l {parent=pelvis, loc=<-2.5693 -0.0109 -7.7501>, rot=<0.0012 0.0572 0.0155 -0.9982>}
 calf_l {parent=thigh_l, loc=<-39.0484 0 0>, rot=<0.0006 -0.0161 -0.0391 -0.9991>}
 calf_twist_01_l {parent=calf_l, loc=<-35.813 -0.0967 0.0397>, rot=<0.0237 0.0095 -0.0035 -0.9997>}
 calf_twist_02_l {parent=calf_l, loc=<-18.8064 -0.0967 0.0397>, rot=<0.0237 0.0095 -0.0035 -0.9997>}
 foot_l {parent=calf_l, loc=<-42.561 0.1077 0.4292>, rot=<0.0418 -0.0455 0.0343 -0.9975>}
 ball_l {parent=foot_l, loc=<-6.1599 13.8264 0.6806>, rot=<0.0001 -0.0001 -0.7063 -0.7079>}
 calf_spike_l {parent=calf_l, loc=<0 0 0>, rot=<0 0 0 -1>}
 deform_knee_bk_01_l {parent=calf_l, loc=<0.6022 -1.0607 0>, rot=<0 0 -0.0408 -0.9992>}
 deform_knee_fr_01_l {parent=calf_l, loc=<-0.2342 0 0>, rot=<0 0 0 1>}
 thigh_twist_01_l {parent=thigh_l, loc=<-21.6303 0 0>, rot=<0.0013 0 0 -1>}
 shorts_fr_l {parent=thigh_twist_01_l, loc=<9.0028 8.1316 0.0152>, rot=<0 0 1 0>}
 shorts_bk_l {parent=thigh_twist_01_l, loc=<8.8248 -7.9238 2.1251>, rot=<0 -1 0 0>}
 shorts_in_l {parent=thigh_twist_01_l, loc=<8.5045 1.5928 8.3501>, rot=<0 -0.7071 0.7071 0>}
 deform_thigh_fr_l {parent=thigh_twist_01_l, loc=<21.5084 0.2281 -0.0598>, rot=<-0.0013 -0.0217 0 -0.9998>}
 deform_thigh_bk_l {parent=thigh_twist_01_l, loc=<16.9514 0.2058 0.1808>, rot=<-0.0013 -0.0217 0 -0.9998>}
 shorts_out_l {parent=thigh_twist_01_l, loc=<9.424 1.1196 -6.4181>, rot=<0 0.7062 0.708 0>}
 thigh_quad_1_l {parent=thigh_twist_01_l, loc=<12.2823 8.0734 0.0612>, rot=<0 0 0 1>}
 thigh_quad_l {parent=thigh_l, loc=<-30 0 0>, rot=<0 0 0 1>}
 thigh_r {parent=pelvis, loc=<-2.5691 -0.0109 7.7501>, rot=<0.0572 -0.0012 0.9982 0.0155>}
 calf_r {parent=thigh_r, loc=<39.0487 0 0>, rot=<0.0006 -0.0161 -0.0391 -0.9991>}
 calf_twist_01_r {parent=calf_r, loc=<35.8132 0.0967 -0.0397>, rot=<0.0237 0.0095 -0.0035 -0.9997>}
 calf_twist_02_r {parent=calf_r, loc=<18.8064 0.0967 -0.0397>, rot=<0.0237 0.0095 -0.0035 -0.9997>}
 foot_r {parent=calf_r, loc=<42.5612 -0.1077 -0.4292>, rot=<0.0418 -0.0455 0.0343 -0.9975>}
 ball_r {parent=foot_r, loc=<6.1598 -13.8264 -0.6806>, rot=<0.0001 -0.0001 -0.7063 -0.7079>}
 calf_spike_r {parent=calf_r, loc=<0 0 0>, rot=<0 0 1 0>}
 deform_knee_fr_01_r {parent=calf_r, loc=<0.2341 0 0>, rot=<0 0 0 1>}
 deform_knee_bk_01_r {parent=calf_r, loc=<-0.6022 1.0607 0>, rot=<0 0 -0.0408 -0.9992>}
 thigh_twist_01_r {parent=thigh_r, loc=<21.6303 0 0>, rot=<0.0013 0 0 -1>}
 shorts_fr_r {parent=thigh_twist_01_r, loc=<-9.0026 -8.1316 -0.0152>, rot=<0 0 0 1>}
 shorts_in_r {parent=thigh_twist_01_r, loc=<-8.5043 -1.5928 -8.3501>, rot=<-0.7071 0 0 -0.7071>}
 shorts_bk_r {parent=thigh_twist_01_r, loc=<-8.8246 7.9238 -2.1251>, rot=<1 0 0 0>}
 deform_thigh_bk_r {parent=thigh_twist_01_r, loc=<-16.9512 -0.2058 -0.1808>, rot=<-0.0013 -0.0217 0 -0.9998>}
 deform_thigh_fr_r {parent=thigh_twist_01_r, loc=<-21.5083 -0.2281 0.0598>, rot=<-0.0013 -0.0217 0 -0.9998>}
 shorts_out_r {parent=thigh_twist_01_r, loc=<-9.4239 -1.1196 6.4181>, rot=<0.7062 0 0 -0.708>}
 thigh_quad_1_r {parent=thigh_twist_01_r, loc=<-12.2821 -8.0734 -0.0612>, rot=<0 0 0 1>}
 thigh_quad_r {parent=thigh_r, loc=<30.0002 0 0>, rot=<0 0 1 0>}
 shirt_pelvis {parent=pelvis, loc=<0 0 0>, rot=<0 0 0 -1>}
 shirt_pelvis_l {parent=shirt_pelvis, loc=<0.2174 -0.7759 -14.1784>, rot=<0 0 0 -1>}
 shirt_pelvis_r {parent=shirt_pelvis, loc=<0.2174 -0.7759 14.1784>, rot=<0 0 0 -1>}
 HT_L1_TransitionCurve {type=Curve, loc=<0 0 0>, rot=<0 0 0 1>}
]
variables [
 shldr_twist_deg_l {type=double, default=0}
 shldr_twist_deg_r {type=double, default=0}
]
