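# Animation Blueprint for the M_MED_Bicycle player body cosmetic, adding corrective secondary deformation to the pose copied from its attached parent mesh.
# The copied pose passes through twist and sculpter Control Rigs, bone corrections, then a final body Control Rig.
# Each update derives knee-bend alphas from calf-to-thigh twist, driving additive calf spike rotations and back-shorts shifts.
# Rotation constraints make inner forearm spikes follow the lower-arm twist bones.
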AnimBlueprint /BRCosmetics/Characters/Player/Male/Medium/Bodies/M_MED_Bicycle/Meshes/M_MED_Bicycle_AnimBP
target_skeleton=&/M_MED_Bicycle_Skeleton
graphs [
 AnimGraph {nodes [
   n1 = Control Rig(Source=Copy Pose From Mesh() {bCopyCurves=true, bUseAttachedParent=true}) {
    alphaBoolBlend {blendOption=Linear}
    controlRigAssetReference {blueprintRigClass=&/M_MED_Bicycle_Deform_CtrlRig_twists.M_MED_Bicycle_Deform_CtrlRig_twists_C}
   }
   n2 = Control Rig!(
    Source=@n1
    Alpha
    KneeRot_fr_l_114deg:Vector
    KneeTrans_fr_l_114deg:Vector="0.258,-3.591,0.009"
    KneeScale_fr_l_114deg:Vector="1,1,1"
    KneeScale_fr_l_130deg:Vector="1,1,1"
    KneeRot_fr_l_130deg:Vector="0,0,-51.2"
    KneeTrans_fr_l_130deg:Vector="0.258,-3.591,0.009"
   ) -> (Pose) {
    alphaBoolBlend {blendOption=Linear}
    controlRigAssetReference {blueprintRigClass=&/M_MED_Bicycle_Deform_CtrlRig_sculpters.M_MED_Bicycle_Deform_CtrlRig_sculpters_C}
   }
   n3 = "Transform (Modify) Bone - Bone: calf_spike_l"(
    ComponentPose=Local To Component(LocalPose=@n2)
    Alpha=Get kneeBend_L() -> (kneeBend_L:double)
    Translation="0,0,0"
    Rotation="10,-5,-20"
    Scale="1,1,1"
   ) {
    rotationMode=BMM_Additive
    rotationSpace=BCS_BoneSpace
    alphaBoolBlend {blendOption=Linear}
    boneToModify {boneName=calf_spike_l}
    rotation=<rot 10 -5 -20>
   }
   n4 = "Transform (Modify) Bone - Bone: calf_spike_r"(ComponentPose=@n3, Alpha=Get kneeBend_R() -> (kneeBend_R:double), Translation="0,0,0", Rotation="-10,5,20", Scale="1,1,1") {
    rotationMode=BMM_Additive
    rotationSpace=BCS_BoneSpace
    alphaBoolBlend {blendOption=Linear}
    boneToModify {boneName=calf_spike_r}
    rotation=<rot -10 5 20>
   }
   n5 = "Transform (Modify) Bone - Bone: shorts_bk_l"(ComponentPose=@n4, Alpha=Get kneeBend_L(), Translation="4,0,0", Rotation="0,0,0", Scale="1,1,1") {
    rotationMode=BMM_Additive
    rotationSpace=BCS_ParentBoneSpace
    translationMode=BMM_Additive
    translationSpace=BCS_ParentBoneSpace
    alphaBoolBlend {blendOption=Linear}
    boneToModify {boneName=shorts_bk_l}
    translation {x=4}
   }
   n6 = "Transform (Modify) Bone - Bone: shorts_bk_r"(ComponentPose=@n5, Alpha=Get kneeBend_R(), Translation="-4,0,0", Rotation="0,0,0", Scale="1,1,1") {
    rotationMode=BMM_Additive
    rotationSpace=BCS_ParentBoneSpace
    translationMode=BMM_Additive
    translationSpace=BCS_ParentBoneSpace
    alphaBoolBlend {blendOption=Linear}
    boneToModify {boneName=shorts_bk_r}
    translation {x=-4}
   }
   n7 = "Constraint - forearm_spike_in_r (R)"(ComponentPose=@n6, Alpha=1, ConstraintWeights_0=1) {
    alphaBoolBlend {blendOption=Linear}
    boneToModify {boneName=forearm_spike_in_r}
    constraintSetup [{transformType=Rotation, targetBone {boneName=lowerarm_twist_02_r}}]
    constraintWeights [1]
   }
   n8 = "Constraint - forearm_spike_in_l (R)"(ComponentPose=@n7, Alpha=1, ConstraintWeights_0=1) {
    alphaBoolBlend {blendOption=Linear}
    boneToModify {boneName=forearm_spike_in_l}
    constraintSetup [{transformType=Rotation, targetBone {boneName=lowerarm_twist_02_l}}]
    constraintWeights [1]
   }
   n9 = Control Rig!(Source=Component To Local(ComponentPose=@n8), Alpha) -> (Pose) {
    alphaBoolBlend {blendOption=Linear}
    controlRigAssetReference {blueprintRigClass=&/M_MED_Bicycle_CtrlRig.M_MED_Bicycle_CtrlRig_C}
   }
   Output Pose(Result=@n9) {layerGroup=None}
  ]}
 EventGraph {nodes [
   Event BlueprintUpdateAnimation()
   > Set kneeBend_L(kneeBend_L:double=GetAlphaFromReferenceTransformDelta(
     SkeletalMeshComponent=Get Body_Mesh()
     BoneName=calf_l
     BaseName=thigh_l
     DeltaComponentType=QuaternionTwist
     InputRange=(X=-15,Y=-100)
     TwistAxis="0,0,1"
    )) -> (Output_Get:double)
   > Set kneeBend_R(kneeBend_R:double=GetAlphaFromReferenceTransformDelta(
     SkeletalMeshComponent=Get Body_Mesh()
     BoneName=calf_r
     BaseName=thigh_r
     DeltaComponentType=QuaternionTwist
     InputRange=(X=-15,Y=-100)
     TwistAxis="0,0,1"
    )) -> (Output_Get:double)
   Event BlueprintInitializeAnimation()
   > Set Body_Mesh(Body_Mesh=GetOwnerSkeletalMeshComponent())
   > Set Pawn(Pawn:Fort Pawn Object Reference=TryGetFortPawn()) -> (Output_Get:Fort Pawn Object Reference)
  ]}
]
variables [
 arm_cst_alpha {type=number, default=0.65}
 Body_Mesh {type=SkeletalMeshComponent}
 kneeBend_L {type=number}
 kneeBend_R {type=number}
 Pawn {type=FortPawn, default=null}
 sleeve_dyn_alpha {type=number}
]
